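; Behavior tree decorator that checks whether an AI-controlled character is allowed to sleep.
; Its AI condition check casts the owning controller to FortAIController and queries IsAllowedtoSleep on it.
; The branch and cast-failure paths feed a MergeExecution node, whose true-input flag becomes the condition result.
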(fn PerformConditionCheckAI
  (params (OwnerController :type AIController%20Object%20Reference) (ControlledPawn :type Pawn%20Object%20Reference))
  (bind _asfort_aicontroller (Utilities|Casting|CastToFortAIController OwnerController)
    (:then
      (bind _returnvalue (AI|IsAllowedtoSleep _asfort_aicontroller))
      (if _returnvalue
        _))
    (:CastFailed
      _))
  (bind _fromtrueinput (Utilities|FlowControl|MergeExecution)
    (:Then
      (return _fromtrueinput))))
