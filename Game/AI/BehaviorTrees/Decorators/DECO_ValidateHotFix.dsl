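; Behavior tree decorator that gates a branch on a ScalableFloat value read as a boolean, apparently as a hotfix toggle.
; Both the generic and AI condition checks return the GetValueToCheck variable converted with GetValueasBool.
; The owner actor, AI controller and pawn inputs are ignored, so the result depends only on that configured value.
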(fn PerformConditionCheck
  (params (OwnerActor :type Actor%20Object%20Reference))
  (return (ScalableFloat|GetValueasBool (Variables|Default|GetValueToCheck))))

(fn PerformConditionCheckAI
  (params (OwnerController :type AIController%20Object%20Reference) (ControlledPawn :type Pawn%20Object%20Reference))
  (return (ScalableFloat|GetValueasBool (Variables|Default|GetValueToCheck))))
