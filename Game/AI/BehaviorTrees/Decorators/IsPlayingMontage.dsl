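; IsPlayingMontage is a Behavior Tree decorator that checks whether the AI-controlled pawn is playing an animation montage.
; Its AI condition check casts the controlled pawn to FortPawn and returns the result of HasCurrentMontage on it.
; Both cast branches are empty and the owning controller is not used.
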(fn PerformConditionCheckAI
  (params (OwnerController :type AIController%20Object%20Reference) (ControlledPawn :type Pawn%20Object%20Reference))
  (bind _asfort_pawn (Utilities|Casting|CastToFortPawn ControlledPawn)
    (:then
      _)
    (:CastFailed
      _))
  (return (Animation|HasCurrentMontage _asfort_pawn)))
